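; Grind rail bridge variant that builds vertical supports beneath its spline in the construction script or via the AddSupports event.
; It samples a hardcoded index range at spacing set by spline length over support count, storing local-space transforms.
; A debug-drawn downward trace sizes each scaled static mesh component, which gets a red material.
; RemoveAllSupports destroys tracked and box-overlapped BSG_TrackSupport and BSG_TrackSupportLong actors, not those mesh components.
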(event Custom|AddSupports
  (CallFunction|AddSupportsFunction))

(event Custom|RemoveAllSupports
  (for _array_element (Variables|Default|GetAllSpawnedSupports)
    (Actor|DestroyActor _array_element))
  (for _array_element_2 (Variables|Default|GetAllSpawnLongSupports)
    (Actor|DestroyActor _array_element_2))
  (for _array_element_3 (Variables|Default|GetAllSupportTransforms)
    (bind _outactors (Collision|BoxOverlapActors (.location _array_element_3) "5000.000000,5000.000000,5000.000000" (Utilities|Array|MakeArray "ObjectTypeQuery1") "/GrindRail/Meshes/Proto/BSG_TrackSupport.BSG_TrackSupport_C" :!class /Script/Engine.KismetSystemLibrary))
    (for _array_element_4 _outactors
      (Actor|DestroyActor _array_element_4))
    (Utilities|Array|Clear (Variables|Default|GetAllSpawnedSupports))
    (for _array_element_5 (Variables|Default|GetAllSupportTransforms)
      (bind _outactors_2 (Collision|BoxOverlapActors (.location _array_element_5) "5000.000000,5000.000000,5000.000000" (Utilities|Array|MakeArray "ObjectTypeQuery1") "/GrindRail/Meshes/Proto/BSG_TrackSupportLong.BSG_TrackSupportLong_C" :!class /Script/Engine.KismetSystemLibrary))
      (for _array_element_6 _outactors_2
        (Actor|DestroyActor _array_element_6))
      (Utilities|Array|Clear (Variables|Default|GetAllSpawnLongSupports)))))

(fn AddSupportsFunction
  (Variables|Default|SetLengthAlongSplineAtAdd (/ (Spline|GetSplineLength (Variables|Default|GetSpline)) (Variables|Default|GetNumVericalSupports)))
  (for _index (range 1 11)
    (Utilities|Array|AddUnique (Variables|Default|GetAllSupportTransforms) (Spline|GetTransformatDistanceAlongSpline (Variables|Default|GetSpline) (* _index (Variables|Default|GetLengthAlongSplineAtAdd)) "Local" false)))
  (for _array_element (Variables|Default|GetAllSupportTransforms)
    (bind (_supportlength _startlocation) (CallFunction|TraceSupportHeight (.location _array_element)))
    (bind _returnvalue (AddComponent|NODE_AddStaticMeshComponent-2 :RelativeTransform (Math|Transform|MakeTransform (.location (Math|Transform|MakeTransform _startlocation "0, 0, 0" (Math|Vector|MakeVector 1.5 1.5 _supportlength))) "0, 0, 0" (.scale (Math|Transform|MakeTransform _startlocation "0, 0, 0" (Math|Vector|MakeVector 1.5 1.5 _supportlength))))))
    (Rendering|Material|SetMaterial _returnvalue 0 "/Game/Characters/Player/Male/Medium/Bodies/M_MED_ArnoldArchetypeSwift_NPC/Materials/Swift_Pants_Red.Swift_Pants_Red")))

(fn TraceSupportHeight
  (params (Location :type Vector))
  (bind _returnvalue (- Location "0.000000,0.000000,75.000000"))
  (bind _outhit (Collision|LineTraceByChannel _returnvalue (- Location "0.000000,0.000000,15000.000000") "TraceTypeQuery10" true (Utilities|Array|MakeArray self) "ForDuration" true "(R=1.000000,G=0.000000,B=0.000000,A=1.000000)" "(R=0.000000,G=1.000000,B=0.000000,A=1.000000)" 1e+01))
  (return (/ (- (.z _returnvalue) (- (.z (.Location (Collision|BreakHitResult _outhit))) (Variables|Default|GetVerticalSupportAddition))) 2.0) _returnvalue))

(fn ConstructionScript
  (|CallParentFunction)
  (CallFunction|AddSupportsFunction))
