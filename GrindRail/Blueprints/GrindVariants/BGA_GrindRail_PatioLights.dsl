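; Grind rail variant that strings patio or Christmas lights along its spline and always grants fall damage immunity.
; On BeginPlay, UpdateLights clears both collision-free HISMs, then batch-adds evenly spaced, slightly raised instances to the one LightType selects, with capped draw distance.
; LightTransforms is appended to without being cleared in that function.
; Tip overrides use a grind-wire cap mesh without world position offset evaluation and mirror the end tip in X.
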(event EventBeginPlay
  (|CallParentFunction)
  (CallFunction|UpdateLights))

(event Collision|EventActorBeginOverlap
  (params (OtherActor :type Actor%20Object%20Reference)))

(event EventTick
  (params (DeltaSeconds :type Float%20%28single-precision%29)))

(fn AddTip
  (params (EndTip :type Boolean) (OnlyUpdateSpline :type Boolean) (TipMesh :type Static%20Mesh%20Object%20Reference))
  (|CallParentFunction EndTip OnlyUpdateSpline TipMesh)
  (Transformation|SetWorldScale3D (Variables|Default|GetEndTip) "-1.000000,1.000000,1.000000"))

(fn ShouldApplyFallDamageImmunity
  (return true))

(fn SpawnTipComponent
  (bind _returnvalue (Game|AddComponentbyClass 0 (Variables|Default|GetGrindRail_TipComponent%20Class) false (struct :Location "0, 0, 0" :Rotation "0, 0, 0" :Scale "1.000000,1.000000,1.000000")))
  (bind _asgrind_rail_tip_component (Utilities|Casting|CastToGrindRail_TipComponent _returnvalue))
  (Components|StaticMesh|SetStaticMesh _asgrind_rail_tip_component "/Game/Environments/Helios/Props/Grindables/Meshes/SM_Proto_GrindWire_Cap_01.SM_Proto_GrindWire_Cap_01")
  (Rendering|LOD|SetEvaluateWorldPositionOffset _asgrind_rail_tip_component false :!class /Script/Engine.StaticMeshComponent)
  (return _returnvalue))

(fn UpdateLights
  (bind _lightdistance (Variables|Default|GetLightDistance))
  (Collision|SetCollisionEnabled (Variables|Default|GetChristmasLights_HISM) "NoCollision")
  (Collision|SetCollisionEnabled (Variables|Default|GetPatioLights_HISM) "NoCollision")
  (Components|InstancedStaticMesh|ClearInstances (Variables|Default|GetPatioLights_HISM))
  (Components|InstancedStaticMesh|ClearInstances (Variables|Default|GetChristmasLights_HISM))
  (for _index (range (Math|Float|Floor (Math|Float|SafeDivide (Spline|GetSplineLength (Variables|Default|GetSpline)) _lightdistance)))
    (Utilities|Array|Add (Variables|Default|GetLightTransforms) (Math|Transform|MakeTransform (+ (Spline|GetLocationatDistanceAlongSpline (Variables|Default|GetSpline) (+ (* _index _lightdistance) (* _lightdistance 0.5)) "Local") (Math|Vector|MakeVector 0.0 0.0 15.0))))
    (LOD|SetMaxDrawDistance (Utilities|Select :Index (Variables|Default|GetLightType) :Option%200 (Variables|Default|GetPatioLights_HISM) :Option%201 (Variables|Default|GetChristmasLights_HISM)) 6e+03))
  (bind _returnvalue (Components|InstancedStaticMesh|AddInstances (Utilities|Select :Index (Variables|Default|GetLightType) :Option%200 (Variables|Default|GetPatioLights_HISM) :Option%201 (Variables|Default|GetChristmasLights_HISM)) (Variables|Default|GetLightTransforms) false false true))
  (Variables|Default|SetLightInstances_Old _returnvalue))

(fn ConstructionScript
  (|CallParentFunction))
