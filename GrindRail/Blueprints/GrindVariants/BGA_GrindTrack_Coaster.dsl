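; Coaster variant of the grind rail, building a spline-based track with rusted-track pole supports and end caps.
; Construction script caches spline-point transforms and optionally labels point indices for debugging.
; CreateSplineMesh builds each segment as a solid spline mesh with an attached, scaled interaction mesh, plus editor billboards.
; AddSupports spawns pole actors at cached points not on the exclusion list; SpawnTipComponent swaps in the cap mesh.
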(event Custom|AddSupports
  (for (_array_element _array_index) (Variables|Default|GetAllSupportTransforms)
    (if (not (Utilities|Array|ContainsItem (Variables|Default|GetExcludedSupportsList) _array_index))
      (Variables|Default|SetLocation (.location _array_element))
      (Variables|Default|SetRotation (.rotation _array_element))
      (Game|SpawnActorfromClass "/GrindRail/RustedTrack/Blueprints/BSG_RustedTrack_GrindWirePole_A.BSG_RustedTrack_GrindWirePole_A_C" (Math|Transform|MakeTransform (+ (Variables|Default|GetLocation) "0.000000,0.000000,-15.000000") (Math|Rotator|MakeRotator 0.0 0.0 (+ (.yaw (Variables|Default|GetRotation)) 9e+01))) "AlwaysSpawn" "MultiplyWithRoot" 0 0))))

(fn CreateSplineMesh
  (params (FirstPoint :type Integer) (SecondPoint :type Integer) (MeshEditorSprite :type Texture%202D%20Object%20Reference))
  (bind _firstpointrotation_roll (Math|Rotator|BreakRotator (Variables|GetFirstPointRotation_Roll)))
  (bind _interactionrailmeshcomp (LocalVariables|GetInteractionRailMeshComp))
  (bind _secondpointrotation_roll (Math|Rotator|BreakRotator (Variables|GetSecondPointRotation_Roll)))
  (bind _firstpointrotation_roll_2 (Math|Rotator|BreakRotator (Variables|GetFirstPointRotation_Roll)))
  (bind _secondpointrotation_roll_2 (Math|Rotator|BreakRotator (Variables|GetSecondPointRotation_Roll)))
  (bind _spline (Variables|Default|GetSpline))
  (bind _solidrailmeshcomp (LocalVariables|GetSolidRailMeshComp))
  (LocalVariables|SetFirstPointRotation (Spline|GetRotationatSplinePoint _spline FirstPoint "Local"))
  (LocalVariables|SetSecondPointRotation (Spline|GetRotationatSplinePoint _spline SecondPoint "Local"))
  (LocalVariables|SetFirstPointLocation (+ (Variables|/Script/Engine.SceneComponent|GetRelativeLocation (Variables|Default|GetSpline)) (Spline|GetLocationatSplinePoint _spline FirstPoint "Local")))
  (LocalVariables|SetFirstPointTangent (Spline|GetTangentatSplinePoint _spline FirstPoint "Local"))
  (LocalVariables|SetSecondPointLocation (+ (Variables|/Script/Engine.SceneComponent|GetRelativeLocation (Variables|Default|GetSpline)) (Spline|GetLocationatSplinePoint _spline SecondPoint "Local")))
  (LocalVariables|SetSecndPointTangent (Spline|GetTangentatSplinePoint _spline SecondPoint "Local"))
  (LocalVariables|SetDistanceToStartPoint (Spline|GetDistanceAlongSplineatSplinePoint (Variables|Default|GetSpline) (Variables|GetFirstPoint)))
  (LocalVariables|SetDistanceToEndPoint (Spline|GetDistanceAlongSplineatSplinePoint (Variables|Default|GetSpline) (Variables|GetSecondPoint)))
  (Utilities|FlowControl|Sequence
    (:then_0
      ;[1 1 1] Generate Solid Rail Mesh component
      (bind _splinemeshcomponent (CallFunction|AddSolidRailMeshComponent (Math|Float|Absolute(Float) (- (LocalVariables|GetDistanceToEndPoint) (LocalVariables|GetDistanceToStartPoint))) (Variables|GetFirstPoint) (Variables|GetSecondPoint)))
      (bind _output_get (LocalVariables|SetSolidRailMeshComp _splinemeshcomponent))
      (Utilities|IsValid _output_get
        (:Is%20Valid
          (CallFunction|UnifiedMaterialAndUV (Variables|GetFirstPoint) (Variables|GetSecondPoint) _output_get) ; material_unifiedmaterial and UV
          (SplineMesh|SetStartandEnd (LocalVariables|GetSolidRailMeshComp) (LocalVariables|GetFirstPointLocation) (LocalVariables|GetFirstPointTangent) (LocalVariables|GetSecondPointLocation) (LocalVariables|GetSecndPointTangent) true :!class /Script/Engine.SplineMeshComponent)
          (SplineMesh|SetStartRollDegrees _solidrailmeshcomp _firstpointrotation_roll_2 true)
          (SplineMesh|SetEndRollDegrees _solidrailmeshcomp _secondpointrotation_roll_2 true)
          (CallFunction|SetupMeshInfo (LocalVariables|GetSolidRailMeshComp))
          ;;[1 0.1264 0.697] Matrerial
          (Utilities|Array|Add (Variables|Default|GetSplineStaticMesh) (LocalVariables|GetSolidRailMeshComp)))))
    (:then_1
      ;[1 1 1] Generate Interaction mesh component
      (bind _splinemeshcomponent_2 (CallFunction|AddInteractionRailMeshComponent))
      (bind _output_get_2 (LocalVariables|SetInteractionRailMeshComp _splinemeshcomponent_2))
      (Transformation|SetWorldScale3D _output_get_2 (* (Transformation|GetWorldScale _output_get_2) (Variables|Default|GetInteractionScale)))
      (Utilities|IsValid (LocalVariables|GetInteractionRailMeshComp)
        (:Is%20Valid
          (Transformation|AttachComponentToComponent (LocalVariables|GetInteractionRailMeshComp) (LocalVariables|GetSolidRailMeshComp) "None" "KeepRelative" "KeepRelative" "KeepRelative" false)
          (SplineMesh|SetStartandEnd (LocalVariables|GetInteractionRailMeshComp) (LocalVariables|GetFirstPointLocation) (LocalVariables|GetFirstPointTangent) (LocalVariables|GetSecondPointLocation) (LocalVariables|GetSecndPointTangent) true :!class /Script/Engine.SplineMeshComponent)
          (SplineMesh|SetStartRollDegrees _interactionrailmeshcomp _firstpointrotation_roll true)
          (SplineMesh|SetEndRollDegrees _interactionrailmeshcomp _secondpointrotation_roll true)
          ;[1 1 1] Editor debug billboard
          (if (Math|Boolean|ANDBoolean (Editor|InEditor) (> (Variables|GetFirstPoint) 0) (Utilities|IsValid (Variables|GetMeshEditorSprite) :!class /Script/Engine.KismetSystemLibrary))
            (bind _returnvalue (AddComponent|NODE_AddBillboardComponent-0 :RelativeTransform_Location "0, 0, 0" :RelativeTransform_Rotation "0, 0, 0" :RelativeTransform_Scale "0.300000,0.300000,0.300000"))
            (Transformation|SetWorldLocation _returnvalue (+ (Spline|GetLocationatSplinePoint (Variables|Default|GetSpline) (Variables|GetFirstPoint) "World") "0.000000,0.000000,50.000000") false true)
            (Rendering|Components|Sprite|SetSprite _returnvalue (Variables|GetMeshEditorSprite))))))))

(fn SpawnTipComponent
  (bind _tipcomponent (|CallParentFunction))
  (bind _asgrind_rail_tip_component (Utilities|Casting|CastToGrindRail_TipComponent _tipcomponent))
  (Components|StaticMesh|SetStaticMesh _asgrind_rail_tip_component "/GrindRail/RustedTrack/Meshes/SM_RustedTrack_GrindWireCap_A.SM_RustedTrack_GrindWireCap_A")
  (Rendering|LOD|SetEvaluateWorldPositionOffset _asgrind_rail_tip_component false :!class /Script/Engine.StaticMeshComponent)
  (return _asgrind_rail_tip_component))

(fn ConstructionScript
  (|CallParentFunction)
  (Utilities|Array|Clear (Variables|Default|GetAllSupportTransforms))
  (for _index (range (+ (Spline|GetNumberOfSplinePoints (Variables|Default|GetSpline)) 1))
    (Utilities|Array|AddUnique (Variables|Default|GetAllSupportTransforms) (Spline|GetTransformatSplinePoint (Variables|Default|GetSpline) _index "World" false)))
  (if (Variables|Default|GetDebugControlPoints)
    (for _array_element (Actor|GetComponentsByClass "/Script/Engine.TextRenderComponent")
      (Components|DestroyComponent _array_element))
    (for _index_2 (range (+ (Spline|GetNumberOfSplinePoints (Variables|Default|GetSpline)) 1))
      (bind _returnvalue (AddComponent|NODE_AddTextRenderComponent-0 :RelativeTransform (Math|Conversions|ToTransform(Vector) (+ (Spline|GetLocationatSplinePoint (Variables|Default|GetSpline) _index_2 "Local") "0.000000,0.000000,150.000000"))))
      (Rendering|Components|TextRender|SetText _returnvalue (Utilities|Text|ToText(Integer) _index_2)))))
