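; Audio manager component for a grind rail coaster that plays music for the locally viewed player while they grind.
; At begin play it caches the coaster's color and listens for players starting to grind, then binds to the local player's grinding provider delegates.
; Grind begin spawns a 2D sound if none is valid, registers it with the music event subsystem and attaches it to the player's mesh.
; Grind end fades it out and removes the event.
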(event EventBeginPlay
  (bind _asbga_grind_track_coaster (Utilities|Casting|CastToBGA_GrindTrack_Coaster (Components|GetOwner)))
  (bind _output_get (Variables|Default|SetAs%20BGA%20Grind%20Track%20Proto _asbga_grind_track_coaster))
  (Variables|Default|SetCurrentCosterColor (Variables|/GrindRail/Blueprints/BGA_GrindRail.BGA_GrindRail_C|GetColor_Select_Enum _output_get))
  (Default|AssignPlayerStartedGrinding _output_get (delegate Custom|PlayerStartedGrinding_Event)))

(event Custom|PlayerStartedGrinding_Event
  (params (GrindingPlayer :type Player%20Pawn%20Athena%20Object%20Reference) (GrindingProvider :type BP%20Fort%20Pawn%20Grinding%20Provider%20Object%20Reference))
  (if (Math|Boolean|ANDBoolean (Pawn|IsLocallyViewed GrindingPlayer) (Utilities|IsValid GrindingProvider :!class /Script/Engine.KismetSystemLibrary) (Utilities|IsValid GrindingProvider :!class /Script/Engine.KismetSystemLibrary))
    (Variables|Default|SetLocalPlayerOnCoaster GrindingPlayer)
    (Variables|Default|SetGrinding%20Provider GrindingProvider)
    (EventDispatchers|AssignBegunGrindingDelegate (Variables|Default|GetGrinding%20Provider) (EventDispatchers|CreateEvent :!event Custom|BegunGrindingDelegate_Event))
    (EventDispatchers|AssignEndedGrindingDelegate (Variables|Default|GetGrinding%20Provider) (EventDispatchers|CreateEvent :!event Custom|EndedGrindingDelegate_Event))))

(event Custom|BegunGrindingDelegate_Event
  (params (bWasAlreadyGrinding :type Boolean) (bWasJumpingFromRail :type Boolean) (bEnteredFromInteraction :type Boolean) (PreviousPlayerLocation :type Vector))
  (Variables|Default|GetMusicComponent
    (:then)
    (:else
      (bind _mxcoaster (Variables|Setup|GetMxCoaster)
        (:then
          (bind _returnvalue (Audio|SpawnSound2D _mxcoaster))
          (bind _output_get (Variables|Default|SetMusicComponent _returnvalue))
          (Audio|AddEventbyComponent (GameInstanceSubsystems|GetMusicEventSubsystem) _output_get (Variables|Default|GetEvent%20Tag) "(GameplayTags=)" 0 "MusicParty")
          (bind _localplayeroncoaster (Variables|Default|GetLocalPlayerOnCoaster)
            (:then
              (bind _musiccomponent (Variables|Default|GetMusicComponent)
                (:then
                  (Transformation|AttachComponentToComponent _musiccomponent (Variables|/Script/Engine.Character|GetMesh _localplayeroncoaster) "None" "SnapToTarget" "SnapToTarget" "KeepRelative" true))
                (:else)))
            (:else)))
        (:else)))))

(event Custom|EndedGrindingDelegate_Event
  (bind _musiccomponent (Variables|Default|GetMusicComponent)
    (:then
      (Audio|Components|Audio|FadeOut _musiccomponent (Variables|Setup|GetFade%20Out%20Duration) 0.0 "Linear" :!class /Script/Engine.AudioComponent)
      (Audio|RemoveEvent (GameInstanceSubsystems|GetMusicEventSubsystem) _musiccomponent))
    (:else)))
